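# Material function for the DenimEquip skydiving contrail that computes a swaying balloon position offset by summing two chained rotations about the particle's world position.
# Each angle is a sine of UV1's G channel plus randomized, speed-scaled time, with amplitude from dynamic-parameter Str and Random.
# The first rotation uses particle-supplied time, the second engine Time.
# Both axes come from dynamic parameters but repeat X in place of Y, leaving Ry and R2y unused.
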MaterialFunction /BRCosmetics/Athena/Effects/Particles/Skydiving/NS_Contrail_DenimEquip/Materials/MF_Contrail_DenimEquip_BalloonMove
graph {nodes [
  n1 = DynamicParameter() {parameterIndex=1, paramNames [R2x,R2y,R2z,Str]}
  n2 = AppendVector(A=AppendVector(A=@n1.R2x, B=@n1.R2x), B=@n1.R2z)
  n3 = MaterialFunctionCall(Float2=TextureCoordinate() {coordinateIndex=1}) {materialFunction=&BreakOutFloat2Components}
  n4 = DynamicParameter() {paramNames [Rx,Ry,Rz,Random]}
  n5 = Multiply(A=ScalarParameter() {defaultValue=2, group=TeriyakiFishSwing, parameterName=Rotate1Speed}, B=Add(A=@n4.Random, B=Time()))
  n6 = Multiply(
   A=Sine(None=Add(A=@n3.G, B=@n5))
   B=Multiply(A=ScalarParameter() {defaultValue=0.01, group=TeriyakiFishSwing, parameterName=Rotate1Str}, B=Multiply(A=@n1.Str, B=@n4.Random))
  )
  n7 = ParticlePositionWS()
  n8 = WorldPosition()
  n9 = DynamicParameter() {paramNames [Rx,Ry,Rz,Random]}
  n10 = DynamicParameter() {parameterIndex=2, paramNames [Time, "", "", ""]}
  n11 = Add(
   A=@n3.G
   B=Multiply(A=ScalarParameter() {defaultValue=2, group=TeriyakiFishSwing, parameterName=Rotate0Speed}, B=Add(A=@n4.Random, B=@n10.Time))
  )
  n12 = ScalarParameter() {defaultValue=0.07, group=TeriyakiFishSwing, parameterName=Rotate0Str}
  n13 = Multiply(A=@n1.Str, B=@n4.Random)
  n14 = RotateAboutAxis(
   NormalizedRotationAxis=AppendVector(A=AppendVector(A=@n9.Rx, B=@n9.Rx), B=@n9.Rz)
   RotationAngle=Multiply(A=Sine(None=@n11), B=Multiply(A=@n12, B=@n13))
   PivotPoint=@n7
   Position=@n8.XYZ
  )
  n15 = Add(A=@n8.XYZ, B=@n14)
  FunctionOutput(None=Add(A=RotateAboutAxis(NormalizedRotationAxis=@n2, RotationAngle=@n6, PivotPoint=@n7, Position=@n15), B=@n14))
 ]}
