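# Translucent unlit two-sided material for Niagara mesh particles, drawing a string strand in the DenimEquip skydiving contrail.
# A rotated, tiled string texture, tinted by particle color and a second string map, feeds emissive adjustments, masked by a reused Hibiscus vine texture.
# That mask, U-scaled by dynamic parameter MoveUV, also drives a depth fade.
# SplineThicken takes SplineWidth for both base and tip width, leaving SplineTip unused.
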Material /BRCosmetics/Athena/Effects/Particles/Skydiving/NS_Contrail_DenimEquip/Materials/M_Contrail_DenimEquip_SplineString
graph {
 properties {
  bAutomaticallySetUsageInEditor=false
  bUsedWithNiagaraMeshParticles=true
  blendMode=BLEND_Translucent
  pixelDepthOffsetMode=PDOM_Legacy
  shadingModel=MSM_Unlit
  twoSided=true
  usedShadingModels=Unlit
 }
 nodes [
  n1 = DynamicParameter() {paramNames [SplineWidth,SplineTip,MoveUV,Param4]}
  n2 = ScalarParameter() {defaultValue=1, parameterName=Param_1}
  n3 = TextureSample(UVs=Multiply(A=TextureCoordinate(), B=AppendVector(A=@n1.MoveUV, B=@n2))) {texture=/BRCosmetics/Athena/Effects/Particles/Skydiving/NS_Contrail_Hibiscus/Textures/T_Hibiscus_Vine}
  n4 = Saturate(None=Multiply(A=@n3.R, B=ScalarParameter() {defaultValue=2, parameterName=Param_2}) {constB=3})
  DepthFade(Opacity=@n4, FadeDistance=ScalarParameter() {defaultValue=10, parameterName=DepthFade})
  n5 = MaterialFunctionCall(UVs=TextureCoordinate() {vTiling=-10}, Rotation Center=Constant2Vector() {g=0.5, r=0.5}, "Rotation Angle (0-1)"=Constant() {r=-0.25}) {materialFunction=&CustomRotator}
  n6 = TextureSample(UVs=@n5.Rotated Values) {texture=/BRCosmetics/Athena/Effects/Particles/Skydiving/NS_Contrail_DenimEquip/Textures/T_DenimEquip_TilingString}
  n7 = ParticleColor()
  n8 = Power(Base=TextureSample(UVs=@n5.Rotated Values) {texture=/BRCosmetics/Athena/Effects/Particles/Skydiving/NS_Contrail_DenimEquip/Textures/T_DenimEquip_TilingString_FN}.RGB) {constExponent=0.5}
  MaterialFunctionCall(Emissive=Multiply(A=Multiply(A=@n6.RGB, B=@n7.RGB), B=Multiply(A=@n4, B=@n8))) {materialFunction=&MF_FX_EmissiveAdjustments}
  n9 = DynamicParameter() {paramNames [SplineWidth,SplineTip,MoveUV,Param4]}
  MaterialFunctionCall(WidthBase=@n9.SplineWidth, WidthTip=@n9.SplineWidth, WorldPosition=WorldPosition().XYZ) {materialFunction=&SplineThicken}
  n10 = ScalarParameter() {defaultValue=1, parameterName=Param_1}
  n11 = TextureSample(UVs=Multiply(A=TextureCoordinate(), B=AppendVector(A=@n1.MoveUV, B=@n10))) {
   samplerType=SAMPLERTYPE_LinearColor
   texture=/BRCosmetics/Athena/Effects/Particles/Skydiving/NS_Contrail_Hibiscus/Textures/T_Hibiscus_Vine_AO
  }
  Multiply(A=@n11.RGB, B=Multiply(A=@n3.R, B=@n7.RGB))
  ScalarParameter() {defaultValue=15}
  ScalarParameter() {defaultValue=1, parameterName=s}
  ScalarParameter() {defaultValue=1, parameterName=s}
  ScalarParameter() {defaultValue=9, parameterName=Param_3}
 ]
}
