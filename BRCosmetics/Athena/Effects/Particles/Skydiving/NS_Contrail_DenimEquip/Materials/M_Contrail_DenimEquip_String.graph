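# Masked unlit material for Niagara ribbons that draws the tiling string in the DenimEquip skydiving contrail.
# A vine mask borrowed from the Hibiscus contrail is rotated a quarter turn and panned over time, then multiplied with two tiled DenimEquip string textures.
# Particle color tints the result before it feeds MF_FX_EmissiveAdjustments.
# Other branches compute vine-mask fades along U, a gray gradient and a per-particle random value, with no output links shown.
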Material /BRCosmetics/Athena/Effects/Particles/Skydiving/NS_Contrail_DenimEquip/Materials/M_Contrail_DenimEquip_String
graph {
 properties {
  bCastRayTracedShadows=false
  bUsedWithNiagaraRibbons=true
  blendMode=BLEND_Masked
  pixelDepthOffsetMode=PDOM_Legacy
  shadingModel=MSM_Unlit
  usedShadingModels=Unlit
 }
 nodes [
  Clamp(None=Add(A=Constant() {r=1.5}, B=ParticleRandom())) {maxDefault=0.4, minDefault=0.15}
  n1 = Multiply(A=ComponentMask(None=TextureCoordinate()) {r=true})
  LinearInterpolate(
   A=Constant3Vector() {constant=<rgba 0.091 0.091 0.091 1>}
   B=VectorParameter() {defaultValue=<rgba 0.0438 0.0438 0.0438 1>}.RGB
   Alpha=Power(Base=OneMinus(None=@n1), Exp=ScalarParameter() {defaultValue=4}) {constExponent=9}
  )
  n2 = ParticleColor()
  n3 = Panner(
   Coordinate=MaterialFunctionCall(UVs=TextureCoordinate() {uTiling=2}, Rotation Center=Constant2Vector() {g=0.5, r=0.5}, "Rotation Angle (0-1)"=NamedRerouteUsage()) {materialFunction=&CustomRotator}.Rotated Values
   Time=Time()
  ) {speedY=2}
  n4 = Saturate(None=TextureSample(UVs=@n3) {texture=/BRCosmetics/Athena/Effects/Particles/Skydiving/NS_Contrail_Hibiscus/Textures/T_Hibiscus_Vine}.R)
  n5 = TextureCoordinate() {uTiling=-7}
  n6 = Multiply(
   A=@n4
   B=TextureSample(UVs=@n5) {texture=/BRCosmetics/Athena/Effects/Particles/Skydiving/NS_Contrail_DenimEquip/Textures/T_DenimEquip_TilingString}.RGB
  )
  MaterialFunctionCall(Emissive=Multiply(
    A=Multiply(A=@n2.RGB, B=@n6)
    B=TextureSample(UVs=@n5) {texture=/BRCosmetics/Athena/Effects/Particles/Skydiving/NS_Contrail_DenimEquip/Textures/T_DenimEquip_TilingString_FN}.RGB
   )) {materialFunction=&MF_FX_EmissiveAdjustments}
  n7 = TextureCoordinate()
  n8 = Power(Base=OneMinus(None=ComponentMask(None=@n7) {r=true})) {constExponent=0.25}
  Multiply(A=@n4, B=@n8)
  Multiply(A=Power(Base=@n4) {constExponent=0.5}, B=@n8)
  NamedRerouteDeclaration(None=Constant() {r=-0.25}) {name=Rotate, nodeColor=<rgba 1 0 0.0469 1>}
  Panner(Coordinate=@n5) {speedX=-3}
  TextureSample(UVs=@n3) {
   samplerType=SAMPLERTYPE_LinearColor
   texture=/BRCosmetics/Athena/Effects/Particles/Skydiving/NS_Contrail_Hibiscus/Textures/T_Hibiscus_Vine_AO
  }
 ]
}
