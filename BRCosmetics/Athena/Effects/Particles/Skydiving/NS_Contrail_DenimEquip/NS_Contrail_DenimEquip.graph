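# Niagara skydiving contrail for the DenimEquip cosmetic, attaching balloon meshes, hand-wrap meshes and two ribbon strings to the character's skeletal mesh.
# Mesh particles pick bones or sockets by partition and build per-side orientations from Dive Value, with extra balloon tilt in front-end preview.
# Balloons feed rotation axes and a dive-scaled timer into dynamic material parameters.
# Ribbons trail along the setup direction with dive-scaled waveform sway and time-of-day color scaling.
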NiagaraSystem /BRCosmetics/Athena/Effects/Particles/Skydiving/NS_Contrail_DenimEquip/NS_Contrail_DenimEquip
dependencies {
 usedDataInterfaces [
  &NiagaraDataInterfaceArrayFloat
  &NiagaraDataInterfaceArrayFloat3
  &NiagaraDataInterfaceVectorCurve
  &NiagaraDataInterfaceCurve
  &NiagaraDataInterfaceSkeletalMesh
 ]
 usedDynamicInputs [&SetBoolByFloatComparison]
 usedModules [
  &self:SystemSpawnScript.SystemScriptSource.SystemScriptGraph.NiagaraNodeAssignment_0.SetVariables_#1
  &SystemState
  /Game/Effects/Niagara/Modules/Contrails/FN_Contrail_Setup
  &FN_Integrator
  &Timeline
 ]
}
emitters {
 HandWraps {
  emitterUpdateScript [
   &EmitterState(
    Inactive Response=::NewEnumerator1
    Life Cycle Mode=::NewEnumerator1
    Loop Behavior=::NewEnumerator1
    Loop Delay=0.5502
    Loop Duration Mode=::NewEnumerator1
    UseLoopDelay=true
   )
   &SpawnBurst_Instantaneous(Spawn Count=8, Use Spawn Probability=false, UseLoopCountLimit=false)
  ]
  particleSpawnScript [
   &PartitionParticles()
   &InitializeParticle(
    Color=<rgba 0.4769 0.5907 0.486 1>
    Mesh Scale Max=<0.3 0.2 0.3>
    Mesh Scale Min=<0.2 0.2 0.2>
    Mesh Scale Mode=::NewEnumerator2
    UsePositionOffset=false
    Write Material Random=true
   )
   &InitialMeshOrientation(
    Mesh Axis To Orient=<0 0 1>
    Mesh Orientation Mode=::NewEnumerator2
    Orientation Vector=<0 0 0>
    Rotation=&Lerp_Vector(
     A=&RandomRangeVector(Minimum=<0.1 1.1 1>, Maximum=<0.15 1.15 1>)
     B=&RandomRangeVector(Minimum=<0.1 0.9 1>, Maximum=<0.15 1 1>)
     Alpha=&Multiply_FloatByInt(Float=1, Integer=$Emitter.SpawnGroup)
    )
    Rotation Coordinate Space=::NewEnumerator0
    Use Rotation Vector=false
   ) {enabled=false}
  ]
  particleUpdateScript [
   &ParticleState(Kill Particles When Lifetime Has Elapsed=false, Let Infinitely Lived Particles Die When Emitter Deactivates=true)
   &SkeletalMeshLocation(
    Bone / Socket Index=$Particles.PartitionParticles.Partition
    Bone Sampling Mode=::NewEnumerator4
    MaskSpawnGroup=false
    Skeletal Mesh=$System.Skeletal Mesh
   )
   /SetVariables(Particles.Balloon_Orientation:Quat4f=&MultiplyQuaternion(
     Quaternion A=&MakeQuatFromAxisAngle(
      Axis=<1 0 0>
      AngleInDegrees=&SelectFloatFromArray(Direct Array Index=$Particles.PartitionParticles.Partition, Float Selection Array=DataInterface() {propertyValues {FloatData [90,90]}})
     )
     Quaternion B=&MakeQuatFromAxisAngle(
      Axis=<1 0 0>
      AngleInDegrees=&Lerp_Float(
       A=0
       B=&SelectFloatFromArray(Direct Array Index=$Particles.PartitionParticles.Partition, Float Selection Array=DataInterface() {propertyValues {FloatData [-30,-30]}})
       Alpha=$System.FN_Contrail_Setup.Dive Value
      )
      NormalizeAxis=true
     )
    )) {isSetParameters=true}
   /SetVariables(Particles.Balloon_Orientation:Quat4f=&MultiplyQuaternion(
     Quaternion A=$Particles.Balloon_Orientation
     Quaternion B=&MakeQuatFromAxisAngle(
      Axis=&SelectVectorFromArray(
       Direct Array Index=$Particles.PartitionParticles.Partition
       Vector Selection Array=DataInterface() {propertyValues {FloatData [<0 0 0>, <0 0 0>]}}
      )
      AngleInDegrees=&Lerp_Float(
       A=0
       B=&MakeCustomFloatFromBool(
        Bool=$User.bIsFrontEndPreview
        True Float=&SelectFloatFromArray(Direct Array Index=$Particles.PartitionParticles.Partition, Float Selection Array=DataInterface() {propertyValues {FloatData [0,0]}})
        False Float=0
       )
       Alpha=$System.FN_Contrail_Setup.Dive Value
      )
      NormalizeAxis=true
     )
    )) {isSetParameters=true}
   /SetVariables(Particles.Balloon_Position:NiagaraPosition=$Particles.SkeletalMeshLocation.SampledPosition) {isSetParameters=true}
   &DynamicMaterialParameters(
    Index 0 Param 1=&RandomRangeFloat(Minimum=5, Maximum=8)
    Index 0 Param 2=&RandomRangeFloat(Minimum=5, Maximum=8)
    Index 0 Param 3=1
    Module.Param0WriteEnabled=true
    Module.Param1WriteEnabled=true
    Module.Param2WriteEnabled=true
    Module.Param3WriteEnabled=true
   ) {module=DynamicMaterialParameters002}
   &CameraOffset(Camera Offset Amount=2)
   &SolveForcesAndVelocity(Clamp Velocity=false, Limit Acceleration=false)
  ]
  rendererClasses [&NiagaraMeshRendererProperties]
  renderers [{rendererClass=&NiagaraMeshRendererProperties, rendererIndex=0}]
 }
 NEW_Balloons {
  emitterUpdateScript [
   &EmitterState(
    Inactive Response=::NewEnumerator1
    Life Cycle Mode=::NewEnumerator1
    Loop Behavior=::NewEnumerator1
    Loop Duration Mode=::NewEnumerator1
    UseLoopDelay=false
   )
   &SpawnBurst_Instantaneous(Spawn Count=2, Use Spawn Probability=false, UseLoopCountLimit=false)
  ]
  particleSpawnScript [
   &PartitionParticles()
   &InitializeParticle(
    Color Mode=::NewEnumerator0
    Lifetime=5
    Mesh Scale Mode=::NewEnumerator3
    Mesh Uniform Scale=1.45
    UsePositionOffset=false
    Write Material Random=true
   )
   /SetVariables(Particles.VisibilityTag=$Particles.PartitionParticles.Partition) {isSetParameters=true}
  ]
  particleUpdateScript [
   &ParticleState(Kill Particles When Lifetime Has Elapsed=false, Let Infinitely Lived Particles Die When Emitter Deactivates=true)
   &SkeletalMeshLocation(
    Bone / Socket Index=$Particles.PartitionParticles.Partition
    Bone Sampling Mode=::NewEnumerator4
    MaskSpawnGroup=false
    Position Sampling=::NewEnumerator1
    Sampled Position Offset=<0 0 5>
    Sampled Position Offset Coordinate Space:ENiagaraSkelMeshTransforms=::NewEnumerator1
    Skeletal Mesh=$System.Skeletal Mesh
   )
   /SetVariables(Particles.Balloon_Orientation:Quat4f=&MultiplyQuaternion(
     Quaternion A=&MakeQuatFromAxisAngle(
      Axis=<0 0 1>
      AngleInDegrees=&SelectFloatFromArray(Direct Array Index=$Particles.PartitionParticles.Partition, Float Selection Array=DataInterface() {propertyValues {FloatData [5,-10]}})
     )
     Quaternion B=&MakeQuatFromAxisAngle(
      Axis=<0 0 -1>
      AngleInDegrees=&Lerp_Float(
       A=0
       B=&SelectFloatFromArray(Direct Array Index=$Particles.PartitionParticles.Partition, Float Selection Array=DataInterface() {propertyValues {FloatData [-50,50]}})
       Alpha=$System.FN_Contrail_Setup.Dive Value
      )
      NormalizeAxis=true
     )
    )) {isSetParameters=true}
   /SetVariables(Particles.Balloon_Orientation:Quat4f=&MultiplyQuaternion(
     Quaternion A=$Particles.Balloon_Orientation
     Quaternion B=&MakeQuatFromAxisAngle(
      Axis=&SelectVectorFromArray(
       Direct Array Index=$Particles.PartitionParticles.Partition
       Vector Selection Array=DataInterface() {propertyValues {FloatData [<-1 1 0>, <1 1 0>]}}
      )
      AngleInDegrees=&Lerp_Float(
       A=0
       B=&MakeCustomFloatFromBool(
        Bool=$User.bIsFrontEndPreview
        True Float=&SelectFloatFromArray(Direct Array Index=$Particles.PartitionParticles.Partition, Float Selection Array=DataInterface() {propertyValues {FloatData [-50,50]}})
        False Float=0
       )
       Alpha=$System.FN_Contrail_Setup.Dive Value
      )
      NormalizeAxis=true
     )
    )) {isSetParameters=true}
   /SetVariables(Particles.Balloon_Position:NiagaraPosition=$Particles.SkeletalMeshLocation.SampledPosition) {isSetParameters=true}
   /SetVariables(
    Particles.MaterialRotate1Axis:Vector3f=&TransformVector(
     Vector=&RotateVectorByQuaternion(
      Vector To Rotate=&Lerp_Vector(
       A=&SelectVectorFromArray(
        Direct Array Index=$Particles.PartitionParticles.Partition
        Vector Selection Array=DataInterface() {propertyValues {FloatData [<0.4 0.15 1>, <0.75 0.3 2>]}}
       )
       B=&SelectVectorFromArray(
        Direct Array Index=$Particles.PartitionParticles.Partition
        Vector Selection Array=DataInterface() {propertyValues {FloatData [<1 1 0>, <1 1 0>]}}
       )
       Alpha=$System.Timeline.NormalizedTime
      )
      Quaternion=$Particles.Balloon_Orientation
     )
     Destination Space=::World
    )
    Particles.MaterialRotateAxis:Vector3f=&TransformVector(
     Vector=&RotateVectorByQuaternion(
      Vector To Rotate=&Lerp_Vector(
       A=&SelectVectorFromArray(
        Direct Array Index=$Particles.PartitionParticles.Partition
        Vector Selection Array=DataInterface() {propertyValues {FloatData [<0.75 0.35 0.5>, <0.85 0.35 0.5>]}}
       )
       B=&SelectVectorFromArray(
        Direct Array Index=$Particles.PartitionParticles.Partition
        Vector Selection Array=DataInterface() {propertyValues {FloatData [<1 0.5 0.75>, <1 0.5 0.75>]}}
       )
       Alpha=$System.Timeline.NormalizedTime
      )
      Quaternion=$Particles.Balloon_Orientation
     )
     Destination Space=::World
    )
   ) {isSetParameters=true}
   &FN_Timer(Increment Multiplier:NiagaraFloat=&Lerp_Float(A=1, B=2, Alpha=$System.FN_Contrail_Setup.Dive Value))
   &DynamicMaterialParameters(
    Index 0 Param 1=&MakeFloatFromVector(Vector=$Particles.MaterialRotateAxis, Channel=::NewEnumerator0)
    Index 0 Param 2=&MakeFloatFromVector(Vector=$Particles.MaterialRotateAxis, Channel=::NewEnumerator1)
    Index 0 Param 3=&MakeFloatFromVector(Vector=$Particles.MaterialRotateAxis, Channel=::NewEnumerator2)
    Index 0 Param 4=&RandomRangeFloat(Minimum=0.2, Maximum=0.5)
    Index 1 Param 1=&MakeFloatFromVector(Vector=$Particles.MaterialRotate1Axis, Channel=::NewEnumerator0)
    Index 1 Param 2=&MakeFloatFromVector(Vector=$Particles.MaterialRotate1Axis, Channel=::NewEnumerator1)
    Index 1 Param 3=&MakeFloatFromVector(Vector=$Particles.MaterialRotate1Axis, Channel=::NewEnumerator2)
    Index 1 Param 4=&Lerp_Float(A=0.55, B=0.75, Alpha=$System.FN_Contrail_Setup.Dive Value)
    Index 2 Param 1=$Particles.FN_Timer.Value
    Index 2 Param 2=&Lerp_Float(A=0.75, B=1.25, Alpha=$System.FN_Contrail_Setup.Dive Value)
    Index 2 Param 3=&Lerp_Float(A=0.5, B=1, Alpha=$System.FN_Contrail_Setup.Dive Value)
    Index 2 Param 4=&Lerp_Float(A=0.75, B=2, Alpha=$System.FN_Contrail_Setup.Dive Value)
    Index 3 Param 1=0.85
    Index 3 Param 2=1
    Index 3 Param 3=0.66
    Module.Index0Param0Write=true
    Module.Index0Param1Write=true
    Module.Index0Param2Write=true
    Module.Index0Param3Write=true
    Module.Index1Param0Write=true
    Module.Index1Param1Write=true
    Module.Index1Param2Write=true
    Module.Index1Param3Write=true
    Module.Index2Param0Write=true
    Module.Index2Param1Write=true
    Module.Index2Param2Write=true
    Module.Index2Param3Write=true
    Module.Index3Param0Write=true
    Module.Index3Param1Write=true
    Module.Index3Param2Write=true
    Module.Index3Param3Write=true
    Write Parameter Index 1=true
    Write Parameter Index 2=true
    Write Parameter Index 3=true
   )
  ]
  rendererClasses [&NiagaraMeshRendererProperties]
  renderers [
   {rendererClass=&NiagaraMeshRendererProperties, rendererIndex=0}
   {rendererClass=&NiagaraMeshRendererProperties, rendererIndex=1}
  ]
 }
 String {
  emitterUpdateScript [
   &EmitterState(
    Enable Distance Culling=true
    Life Cycle Mode=::NewEnumerator1
    Loop Behavior=::NewEnumerator1
    Loop Duration Mode=::NewEnumerator1
    Max Distance Response=::SleepAndClearParticles
    Module.bUseMaxDistance=true
    Module.bUseMinDistance=true
    Reset Age On Awaken=true
    Scalability Mode=::NewEnumerator1
    UseLoopDelay=false
   )
   &SpawnBurst_Instantaneous(Spawn Count=5, Spawn Group=1, Use Spawn Probability=false, UseLoopCountLimit=false) {module=SpawnBurst_Instantaneous001}
   &SpawnBurst_Instantaneous(Spawn Count=5, Use Spawn Probability=false, UseLoopCountLimit=false)
  ]
  particleSpawnScript [
   &InitializeParticle(
    Color=<rgba 1.5 1.5 1.5 1>
    Ribbon Width=8
    Ribbon Width Mode=::NewEnumerator1
    Sprite Size Mode=::NewEnumerator3
    UsePositionOffset=false
    Write Material Random=true
   )
   /SetVariables(Particles.RibbonLinkOrder:NiagaraFloat=&MakeFloatFromInt(Int=&ReturnParticleIDAttributes())) {isSetParameters=true, module=SetVariables_#2}
  ]
  particleUpdateScript [
   &ParticleState(Kill Particles When Lifetime Has Elapsed=false, Loop Particles Lifetime=true)
   &SampleSkeletalMesh(
    AdditionalYawPitchRoll=false
    Bone Sampling Mode:ENiagaraBoneSamplingMode=::NewEnumerator4
    Sampled Position Offset:Vector3f=&AddVectorToPosition(
     Position=<0 -2 5>
     Vector=&TransformVector(
      Vector=&VectorFromCurve(
       VectorCurve=DataInterface() {propertyValues {
         XCurve {keys [
           {arriveTangent=4.1785, interpMode=RCIM_Cubic, leaveTangent=4.1785, tangentMode=RCTM_User, time=0, value=0}
           {arriveTangent=1.1179, interpMode=RCIM_Cubic, leaveTangent=1.1179, tangentMode=RCTM_User, time=0.2042, value=0.6228}
           {arriveTangent=0.0224, interpMode=RCIM_Cubic, leaveTangent=0.0224, tangentMode=RCTM_User, time=1, value=1}
          ]}
         YCurve {keys [
           {arriveTangent=4.1785, interpMode=RCIM_Cubic, leaveTangent=4.1785, tangentMode=RCTM_User, time=0, value=0}
           {arriveTangent=1.1179, interpMode=RCIM_Cubic, leaveTangent=1.1179, tangentMode=RCTM_User, time=0.2042, value=0.6228}
           {arriveTangent=0.0224, interpMode=RCIM_Cubic, leaveTangent=0.0224, tangentMode=RCTM_User, time=1, value=1}
          ]}
         ZCurve {keys [
           {arriveTangent=4.1785, interpMode=RCIM_Cubic, leaveTangent=4.1785, tangentMode=RCTM_User, time=0, value=0}
           {arriveTangent=1.1179, interpMode=RCIM_Cubic, leaveTangent=1.1179, tangentMode=RCTM_User, time=0.2042, value=0.6228}
           {arriveTangent=0.0224, interpMode=RCIM_Cubic, leaveTangent=0.0224, tangentMode=RCTM_User, time=1, value=1}
          ]}
        }}
       CurveIndex=$Particles.RibbonLinkOrder
       Scale Curve=<3 5 5>
      )
      Destination Space=::Local
     )
    )
    Skeletal Mesh:NiagaraDataInterfaceSkeletalMesh=$System.Skeletal Mesh
   )
   /SetVariables(Particles.Position=&AddVectorToPosition(
     Position=$Particles.SampleSkeletalMesh.SampledPosition
     Vector=&Multiply_VectorByFloat(
      Vector=$System.FN_Contrail_Setup.Direction Vector
      Float=&FloatFromCurve(
       FloatCurve=DataInterface() {propertyValues {Curve {keys [{tangentMode=RCTM_User, time=0, value=0}, {tangentMode=RCTM_User, time=1, value=1}]}}}
       CurveIndex=&ModuloFloat(
        A=&Add_Float(A=&ReturnNormalizedExecIndex(Normalized Index Includes One=false), B=0)
        B=1
       )
       Scale Curve=&FloatFromCurve(
        FloatCurve=DataInterface() {propertyValues {Curve {keys [{interpMode=RCIM_Cubic, tangentMode=RCTM_User, time=0, value=0}, {interpMode=RCIM_Cubic, tangentMode=RCTM_User, time=1, value=1}]}}}
        CurveIndex=$Engine.System.Age
        Scale Curve=&RandomRangeFloat(Minimum=-26, Maximum=-31)
       )
      )
     )
    )) {isSetParameters=true, module=SetVariables_#3}
   /SetVariables(Particles.Position=&AddVectorToPosition(
     Position=$Particles.Position
     Vector=&Multiply_VectorByFloat(
      Vector=<1 0 0>
      Float=&Waveform(
       Global Input Position=&ReturnNormalizedExecIndex()
       Global Amplitude Scale=&ReturnNormalizedExecIndex()
       "Waveform [1]"=::NewEnumerator0
       "[1] Phase Scale"=-2
       "[1] Frequency"=1
       "[1] Amplitude Min/Max"=<-1 1>
       "[1] Scale"=&Lerp_Float(A=8, B=14, Alpha=$System.FN_Contrail_Setup.Dive Value)
      )
     )
    )) {isSetParameters=true, module=SetVariables_#4}
   /SetVariables(Particles.Position=&AddVectorToPosition(
     Position=$Particles.Position
     Vector=&Multiply_VectorByFloat(
      Vector=<0 1 0>
      Float=&Waveform(
       Global Input Position=&ReturnNormalizedExecIndex()
       Global Amplitude Scale=&ReturnNormalizedExecIndex()
       "Waveform [1]"=::NewEnumerator1
       "[1] Phase Scale"=-2
       "[1] Frequency"=1
       "[1] Amplitude Min/Max"=<-1 1>
       "[1] Scale"=&Lerp_Float(A=10, B=15, Alpha=$System.FN_Contrail_Setup.Dive Value)
      )
     )
    )) {isSetParameters=true, module=SetVariables_#5}
   &ScaleRibbonWidth(Ribbon Width Scale=&FloatFromCurve(
     FloatCurve=DataInterface() {propertyValues {Curve {keys [{interpMode=RCIM_Cubic, tangentMode=RCTM_User, time=0, value=0.5}, {interpMode=RCIM_Cubic, tangentMode=RCTM_User, time=1, value=0.7}]}}}
     CurveIndex=$Particles.RibbonLinkOrder
    ))
   &ScaleColor(
    Scale RGB=&Multiply_VectorByFloat(Vector=<1 1 1>, Float=&FNTimeOfDayScalar(Night Scale=0.5, Morning Scale=1, Evening Scale=0.5))
    ScaleA=true
    ScaleRGB=true
   )
  ]
  rendererClasses [&NiagaraRibbonRendererProperties]
  renderers [{rendererClass=&NiagaraRibbonRendererProperties}]
 }
 String001 {
  emitterUpdateScript [
   &EmitterState(
    Enable Distance Culling=true
    Life Cycle Mode=::NewEnumerator1
    Loop Behavior=::NewEnumerator1
    Loop Duration Mode=::NewEnumerator1
    Max Distance Response=::SleepAndClearParticles
    Module.bUseMaxDistance=true
    Module.bUseMinDistance=true
    Reset Age On Awaken=true
    Scalability Mode=::NewEnumerator1
    UseLoopDelay=false
   )
   &SpawnBurst_Instantaneous(Spawn Count=5, Spawn Group=1, Use Spawn Probability=false, UseLoopCountLimit=false) {module=SpawnBurst_Instantaneous001}
   &SpawnBurst_Instantaneous(Spawn Count=5, Use Spawn Probability=false, UseLoopCountLimit=false)
  ]
  particleSpawnScript [
   &InitializeParticle(
    Color=<rgba 1.5 1.5 1.5 1>
    Ribbon Width=8
    Ribbon Width Mode=::NewEnumerator1
    Sprite Size Mode=::NewEnumerator3
    UsePositionOffset=false
    Write Material Random=true
   )
   /SetVariables(Particles.RibbonLinkOrder:NiagaraFloat=&MakeFloatFromInt(Int=&ReturnParticleIDAttributes())) {isSetParameters=true, module=SetVariables_#2}
  ]
  particleUpdateScript [
   &ParticleState(Kill Particles When Lifetime Has Elapsed=false, Loop Particles Lifetime=true)
   &SampleSkeletalMesh(
    AdditionalYawPitchRoll=false
    Bone / Socket Index=1
    Bone Sampling Mode:ENiagaraBoneSamplingMode=::NewEnumerator4
    Sampled Position Offset:Vector3f=&AddVectorToPosition(
     Position=<0 -2 5>
     Vector=&TransformVector(
      Vector=&VectorFromCurve(
       VectorCurve=DataInterface() {propertyValues {
         XCurve {keys [
           {arriveTangent=4.1785, interpMode=RCIM_Cubic, leaveTangent=4.1785, tangentMode=RCTM_User, time=0, value=0}
           {arriveTangent=1.1179, interpMode=RCIM_Cubic, leaveTangent=1.1179, tangentMode=RCTM_User, time=0.2042, value=0.6228}
           {arriveTangent=0.0224, interpMode=RCIM_Cubic, leaveTangent=0.0224, tangentMode=RCTM_User, time=1, value=1}
          ]}
         YCurve {keys [
           {arriveTangent=4.1785, interpMode=RCIM_Cubic, leaveTangent=4.1785, tangentMode=RCTM_User, time=0, value=0}
           {arriveTangent=1.1179, interpMode=RCIM_Cubic, leaveTangent=1.1179, tangentMode=RCTM_User, time=0.2042, value=0.6228}
           {arriveTangent=0.0224, interpMode=RCIM_Cubic, leaveTangent=0.0224, tangentMode=RCTM_User, time=1, value=1}
          ]}
         ZCurve {keys [
           {arriveTangent=4.1785, interpMode=RCIM_Cubic, leaveTangent=4.1785, tangentMode=RCTM_User, time=0, value=0}
           {arriveTangent=1.1179, interpMode=RCIM_Cubic, leaveTangent=1.1179, tangentMode=RCTM_User, time=0.2042, value=0.6228}
           {arriveTangent=0.0224, interpMode=RCIM_Cubic, leaveTangent=0.0224, tangentMode=RCTM_User, time=1, value=1}
          ]}
        }}
       CurveIndex=$Particles.RibbonLinkOrder
       Scale Curve=<-3 -5 5>
      )
      Destination Space=::Local
     )
    )
    Skeletal Mesh:NiagaraDataInterfaceSkeletalMesh=$System.Skeletal Mesh
   )
   /SetVariables(Particles.Position=&AddVectorToPosition(
     Position=$Particles.SampleSkeletalMesh.SampledPosition
     Vector=&Multiply_VectorByFloat(
      Vector=$System.FN_Contrail_Setup.Direction Vector
      Float=&FloatFromCurve(
       FloatCurve=DataInterface() {propertyValues {Curve {keys [{tangentMode=RCTM_User, time=0, value=0}, {tangentMode=RCTM_User, time=1, value=1}]}}}
       CurveIndex=&ModuloFloat(
        A=&Add_Float(A=&ReturnNormalizedExecIndex(Normalized Index Includes One=false), B=0)
        B=1
       )
       Scale Curve=&FloatFromCurve(
        FloatCurve=DataInterface() {propertyValues {Curve {keys [{interpMode=RCIM_Cubic, tangentMode=RCTM_User, time=0, value=0}, {interpMode=RCIM_Cubic, tangentMode=RCTM_User, time=1, value=1}]}}}
        CurveIndex=$Engine.System.Age
        Scale Curve=&RandomRangeFloat(Minimum=-25, Maximum=-30)
       )
      )
     )
    )) {isSetParameters=true, module=SetVariables_#3}
   /SetVariables(Particles.Position=&AddVectorToPosition(
     Position=$Particles.Position
     Vector=&Multiply_VectorByFloat(
      Vector=<1 0 0>
      Float=&Waveform(
       Global Input Position=&ReturnNormalizedExecIndex()
       Global Amplitude Scale=&ReturnNormalizedExecIndex()
       "Waveform [1]"=::NewEnumerator0
       "[1] Phase Scale"=-2
       "[1] Frequency"=1
       "[1] Amplitude Min/Max"=<-1 1>
       "[1] Scale"=&Lerp_Float(A=8, B=13, Alpha=$System.FN_Contrail_Setup.Dive Value)
      )
     )
    )) {isSetParameters=true, module=SetVariables_#4}
   /SetVariables(Particles.Position=&AddVectorToPosition(
     Position=$Particles.Position
     Vector=&Multiply_VectorByFloat(
      Vector=<0 1 0>
      Float=&Waveform(
       Global Input Position=&ReturnNormalizedExecIndex()
       Global Amplitude Scale=&ReturnNormalizedExecIndex()
       "Waveform [1]"=::NewEnumerator1
       "[1] Phase Scale"=-2
       "[1] Frequency"=1
       "[1] Amplitude Min/Max"=<-1 1>
       "[1] Scale"=&Lerp_Float(A=10, B=15, Alpha=$System.FN_Contrail_Setup.Dive Value)
      )
     )
    )) {isSetParameters=true, module=SetVariables_#5}
   &ScaleRibbonWidth(Ribbon Width Scale=&FloatFromCurve(
     FloatCurve=DataInterface() {propertyValues {Curve {keys [{interpMode=RCIM_Cubic, tangentMode=RCTM_User, time=0, value=0.5}, {interpMode=RCIM_Cubic, tangentMode=RCTM_User, time=1, value=0.7}]}}}
     CurveIndex=$Particles.RibbonLinkOrder
    ))
   &ScaleColor(
    Scale RGB=&Multiply_VectorByFloat(Vector=<1 1 1>, Float=&FNTimeOfDayScalar(Night Scale=0.5, Morning Scale=1, Evening Scale=0.5))
    ScaleA=true
    ScaleRGB=true
   )
  ]
  rendererClasses [&NiagaraRibbonRendererProperties]
  renderers [{rendererClass=&NiagaraRibbonRendererProperties}]
 }
}
summary {
 systemName=NS_Contrail_DenimEquip
 emitters [
  {emitterName=NEW_Balloons, rendererClasses [&NiagaraMeshRendererProperties]}
  {emitterName=HandWraps, rendererClasses [&NiagaraMeshRendererProperties]}
  {emitterName=String, rendererClasses [&NiagaraRibbonRendererProperties]}
  {emitterName=String001, rendererClasses [&NiagaraRibbonRendererProperties]}
 ]
 userVariables [
  User.bIsFrontEnd {type=NiagaraBool}
  User.bIsFrontEndPreview {type=NiagaraBool}
  User.ParaGlideLeanAlpha {type=NiagaraFloat, defaultValue=0, description=""}
 ]
}
